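# Surface material for the milk crate A prop in the CrateBarrel environment set, using Default Lit shading.
# A Multiply node tints the diffuse texture with a constant pale yellow color.
# A second texture sample reads the crate's normal map.
# The tint is baked in as a constant, with no exposed parameters, and automatic usage setting in the editor is off.
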Material /SparksMaps/Environment/Props/CrateBarrel_MilkCrate/M_CrateBarrel_MilkCrate_A
graph {
 properties {bAutomaticallySetUsageInEditor=false, usedShadingModels=Default Lit}
 nodes [
  Multiply(
   A=TextureSample() {texture=/Game/Environments/Asteria/Props/CrateBarrel/Kit_CrateBarrel_MilkCrate_A/Textures/T_CrateBarrel_MilkCrate_A_A_D}.RGBA
   B=Constant4Vector() {constant=<rgba 1 1 0.3217 1>}.RGBA
  )
  TextureSample() {
   samplerType=SAMPLERTYPE_Normal
   texture=/Game/Environments/Asteria/Props/CrateBarrel/Kit_CrateBarrel_MilkCrate_A/Textures/T_CrateBarrel_MilkCrate_A_A_N
  }
 ]
}
